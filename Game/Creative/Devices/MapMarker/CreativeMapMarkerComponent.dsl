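; Creative device component that puts a configurable icon and text marker on the minimap and map.
; Setup functions apply icon, color, visibility and team filter on authority; SetupMapMarkerDetails async-loads the texture and flushes net dormancy.
; Per-player visibility uses the use-restriction component, cached to keep per-tick checks cheap unless legacy validation is on, otherwise the team filter.
; A replicated text format inserts the player name for pawn owners.
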(event Custom|SetupMapMarkerDetails
  (params (Icon%20Data%20Icon%20Texture :type Texture%202D%20Soft%20Object%20Reference) (Icon%20Data%20Icon%20Color :type ECreativeMinimapComponentIconColorType%20Enum :default "White") (IconLinearColor :type Linear%20Color%20Structure :default "(R=0.352941,G=0.352941,B=0.352941,A=1.000000)") (VisibilityOnMap :type E_MapMarker_VisibilityOnMap%20Enum :default "NewEnumerator1") (UseTeamFilter :type Boolean :default true) (TeamFilter :type Integer) (bMarkerIsVisible :type Boolean) (IndicatorText :type Text) (IndicatorTextColor :type Linear%20Color%20Structure :default "(R=0.000000,G=0.000000,B=0.000000,A=0.000000)") (Creative%20Use%20Restriction%20Component :type Creative%20Use%20Restriction%20Component%20Object%20Reference) (IconTextureMaterialParamName :type Name :default "IconTexture") (Icon%20Material :type Material%20Interface%20Soft%20Object%20Reference) (Icon%20Scale :type Float%20%28double-precision%29))
  ;[1 1 1] Legacy will be remove post migration UEFN
  (CallFunction|SetCreativeUseRestrictionComponent Creative%20Use%20Restriction%20Component)
  (Utilities|FlowControl|Sequence
    (:then_0
      (Utilities|SwitchHasAuthority
        (:Authority
          (bind _object (Utilities|AsyncLoadAsset Icon%20Data%20Icon%20Texture)
            (:Completed
              (bind _astexture_2d (Utilities|Casting|CastToTexture2D _object))
              (Variables|SetIconData_IconIndex (Utilities|Struct|MakeCreativeMiniMapComponentIconData (Math|Integer|SelectInt -1 0 (== _astexture_2d "/Game/Athena/UI/InGame/Creative/PropertyEditor/Icon_Picker/Textures/T_UI_IconPicker_EMPTY.T_UI_IconPicker_EMPTY")) _astexture_2d Icon%20Material IconTextureMaterialParamName Icon%20Data%20Icon%20Color IconLinearColor Icon%20Scale))
              (Variables|Default|SetTeamFilter TeamFilter)
              (Variables|Default|SetUseTeamFilter UseTeamFilter)
              (CallFunction|SetMarkerVisibility bMarkerIsVisible)
              (MiniMapData|SetIndicatorTextColor IndicatorTextColor)
              (CallFunction|SetVisibilityonMap VisibilityOnMap))))))
    (:then_1
      (Variables|Default|SetIndicatorTextFormat IndicatorText) ; Default value must be non-empty, so that replication is triggered on clients if empty text is set
      (Networking|FlushNetDormancy (Components|GetOwner)))))

(event Custom|Setup%20Map%20Marker
  (params (Icon%20Data%20Icon%20Texture :type Texture%202D%20Soft%20Object%20Reference) (Icon%20Data%20Icon%20Color :type Linear%20Color%20Structure :default "(R=0.352941,G=0.352941,B=0.352941,A=1.000000)") (VisibilityOnMap :type E_MapMarker_VisibilityOnMap%20Enum :default "NewEnumerator1") (UseTeamFilter :type Boolean :default true) (TeamFilter :type Creative%20Team%20Option%20Structure) (IsMarkerVisible :type Boolean) (IndicatorText :type Text) (IndicatorTextColor :type Linear%20Color%20Structure :default "(R=0.000000,G=0.000000,B=0.000000,A=0.000000)") (Creative%20Use%20Restriction%20Component :type Creative%20Use%20Restriction%20Component%20Object%20Reference) (IconTextureParamName :type Name :default "IconTexture") (Icon%20Material :type Material%20Interface%20Soft%20Object%20Reference) (Icon%20Scale :type Float%20%28double-precision%29 :default 1.0))
  ;[0.7166 1 0.7451] Alternative function that takes in a CreativeTeamColor rather than an integer, and converts it to be compatible with the base setup for this component.
  (CallFunction|SetupMapMarkerDetails Icon%20Data%20Icon%20Texture "None" Icon%20Data%20Icon%20Color VisibilityOnMap UseTeamFilter (Creative|TeamOption|GetLegacyTeamIndex TeamFilter) IsMarkerVisible IndicatorText IndicatorTextColor Creative%20Use%20Restriction%20Component IconTextureParamName Icon%20Material Icon%20Scale))

(fn BP_IsVisibleToPlayer
  (params (PC :type Fort%20Player%20Controller%20Object%20Reference))
  (bind _creativeuserestrictioncomponent (Utilities|CreativeUseRestrictionComponent)
    (:IsValid
      (if (ScalableFloat|GetValueasBool (Variables|Default|GetShouldUseLegacyValidation))
        (bind _valid_controller (CreativeUserRestrictions|ActorHasValidParticipantController _creativeuserestrictioncomponent (Variables|GetPC) true false true))
        (return _valid_controller)
        (else
          ; Use cached variable to do the check on the local client, don't do any heavy logic here since it's executed per tick
          (return (Variables|Default|GetIsVisibleOnUseRestriciton)))))
    (:IsNotValid
      (Utilities|FlowControl|Sequence
        (:then_0
          (if (Variables|Default|GetUseTeamFilter)
            (if (Knobs|Team|IsValidTeamIndex (Variables|Default|GetTeamFilter))
              (bind _pc (Variables|GetPC)
                (:then
                  (LocalVariables|SetbIsVisible (== (Math|Conversions|ToInteger(Byte) (Team|GetActorPvPTeamIndex _pc)) (Knobs|Team|GetTeamValueasIndex (Variables|Default|GetTeamFilter)))))
                (:else)))))
        (:then_1
          (return (LocalVariables|GetbIsVisible)))))))

(fn OnRep_IndicatorTextFormat
  (bind _indicatortextformat (Variables|Default|GetIndicatorTextFormat))
  (Networking|CanExecuteCosmeticEvents
    (:True
      (bind _aspawn (Utilities|Casting|CastToPawn (Components|GetOwner))
        (:then
          (bind _asfort_player_state_zone (Utilities|Casting|CastToFortPlayerStateZone (Variables|/Script/Engine.Pawn|GetPlayerState _aspawn))
            (:then
              (MiniMapData|SetIndicatorText (Utilities|Text|FormatText _indicatortextformat (PlayerState|GetPlayerNameforStreaming _asfort_player_state_zone)))
              (return))
            (:CastFailed
              _)))
        (:CastFailed
          _))
      (MiniMapData|SetIndicatorText _indicatortextformat)
      (return))))

(fn OnRep_IsVisibleOnUseRestriciton
  (Default|CallIsVisibleonUseRestrictionRep self))

(fn SetMarkerVisibility
  (params (bMarkerIsVisible :type Boolean))
  (Utilities|SwitchHasAuthority
    (:Authority
      (MiniMapData|SetMiniMapIndicatorIsVisible bMarkerIsVisible))))

(fn SetupUsingKnobValues
  (params (IconIndex :type Integer) (IconTexture :type Texture%202D%20Object%20Reference) (IconColor :type ECreativeMinimapComponentIconColorType%20Enum :default "White") (VisbilityOnMap :type E_MapMarker_VisibilityOnMap%20Enum :default "NewEnumerator1") (InTeamFilter :type Integer) (bMarkerVisibility :type Boolean :default true) (Text :type Text) (TextColor :type Linear%20Color%20Structure :default "(R=1.000000,G=1.000000,B=1.000000,A=1.000000)") (Creative%20Use%20Restriction :type Creative%20Use%20Restriction%20Component%20Object%20Reference))
  (CallFunction|SetCreativeUseRestrictionComponent Creative%20Use%20Restriction)
  (Utilities|FlowControl|Sequence
    (:then_0
      (Utilities|SwitchHasAuthority
        (:Authority
          (Variables|SetIconData_IconIndex (Utilities|Struct|MakeCreativeMiniMapComponentIconData (Variables|GetIconIndex) (Variables|GetIconTexture) 0 "IconTexture" (Variables|GetIconColor)))
          (Variables|Default|SetTeamFilter (Variables|GetInTeamFilter))
          (CallFunction|SetMarkerVisibility (Variables|GetbMarkerVisibility))
          (MiniMapData|SetIndicatorTextColor (Variables|GetTextColor))
          (CallFunction|SetVisibilityonMap (Variables|GetVisbilityOnMap)))))
    (:then_1
      (MiniMapData|SetIndicatorText (Variables|GetText)))))

(fn SetVisibilityOnMap
  (params (VisibilityOnMap :type E_MapMarker_VisibilityOnMap%20Enum :default "NewEnumerator0"))
  (MiniMapData|SetMiniMapIndicatorIsVisibleonMiniMap (or (Utilities|Enum|Equal(Enum) (Variables|GetVisibilityOnMap) "NewEnumerator1") (Utilities|Enum|Equal(Enum) (Variables|GetVisibilityOnMap) "NewEnumerator2")))
  (MiniMapData|SetMiniMapIndicatorIsVisibleonMap (or (Utilities|Enum|Equal(Enum) (Variables|GetVisibilityOnMap) "NewEnumerator1") (Utilities|Enum|Equal(Enum) (Variables|GetVisibilityOnMap) "NewEnumerator3"))))

(fn SetVisibilityOnUseRestriction
  (params (Actor :type Actor%20Object%20Reference))
  (CallFunction|UpdateIsVisible Actor)
  (if (ScalableFloat|GetValueasBool (Variables|CreativeMiniMapComponent|GetShouldUseNativeValidation))
    _
    (elif (ScalableFloat|GetValueasBool (Variables|Default|GetShouldUseLegacyValidation))
      _
      (else
        (bind _creativeuserestrictioncomponent (Utilities|CreativeUseRestrictionComponent)
          (:IsValid
            (bind _valid_controller (CreativeUserRestrictions|ActorHasValidParticipantController _creativeuserestrictioncomponent Actor true false true))
            (Variables|Default|SetIsVisibleOnUseRestriciton _valid_controller)))))))
