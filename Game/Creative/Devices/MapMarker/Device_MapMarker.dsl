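; Creative map marker device that shows an icon and text label on maps, filtered by team and class, with an optional objective trail.
; Initialization feeds knob values into the use-restriction, enabled and map marker components, with initial visibility depending on whether a minigame is in progress.
; Enable/disable events and minigame start/end set the enabled component, whose state change drives marker visibility.
; Separate events toggle the objective trail for the instigator and team.
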(event EventInitializeDevice
  ;[1 1 1] Defaults
  (|CallParentFunction)
  (Initialization|InitializeComponentWithAffects (Variables|Default|GetCreative_UseRestriction_Component) (Variables|Default|GetFortMinigameProgress) (Variables|Default|GetEnabledComponent) 0 (Variables|Knobs|GetKnob_TeamFilter) (Variables|Knobs|GetClassIndex) 0 1 (== (Variables|Knobs|GetKnob_AffectsTeam) 0) (== (Variables|Knobs|GetKnob_AffectsClass) 0) false)
  (Class|CreativeMapMarkerComponent|SetupMapMarkerDetails (Variables|Default|GetCreativeMapMarkerComponent) (Creative|UI|IconPicker|GetIconLibraryTextureWithEmptyTexture (Variables|Knobs|GetKnob_Icon)) (Variables|Knobs|GetKnob_IconColor) "(R=0.352941,G=0.352941,B=0.352941,A=1.000000)" (Variables|Knobs|GetKnob_ShowOnMaps) true (Variables|Knobs|GetKnob_TeamFilter) (CallFunction|GetInitialMarkerVisibility) (Utilities|Text|ToText(String) (Variables|Knobs|GetKnob_Text)) (Variables|Knobs|GetKnob_TextColor) (Variables|Default|GetCreative_UseRestriction_Component) "IconTexture" 0 0.0))

(event EventOnMinigameStarted
  ;[1 1 1] Defaults
  (|CallParentFunction)
  (Class|CreativeEnabledComponent|SetEnabled (Variables|Default|GetEnabledComponent) (Math|Conversions|ToBoolean(Integer) (Variables|Knobs|GetKnob_EnabledOnGameStart))))

(event EventOnMinigameEnded
  ;[1 1 1] Defaults
  (|CallParentFunction)
  (Class|CreativeEnabledComponent|SetEnabled (Variables|Default|GetEnabledComponent) true))

(event Custom|Enable%20Event
  (params (Instigator :type Controller%20Object%20Reference))
  ;[1 1 1] Triggers & Receivers
  (Class|CreativeEnabledComponent|SetEnabled (Variables|Default|GetEnabledComponent) true))

(event Custom|Disable%20Event
  (params (Instigator :type Controller%20Object%20Reference))
  ;[1 1 1] Triggers & Receivers
  (Class|CreativeEnabledComponent|SetEnabled (Variables|Default|GetEnabledComponent) false))

(event OnEnabledStateChanged(EnabledComponent)
  (params (Enabled :type Boolean))
  ;[1 1 1] Triggers & Receivers
  (Class|CreativeMapMarkerComponent|SetMarkerVisibility (Variables|Default|GetCreativeMapMarkerComponent) Enabled))

(event Custom|Activate%20Objective%20Trail
  (params (Instigator :type Controller%20Object%20Reference))
  ;[1 1 1] Objective Trail Component
  (Class|CreativeObjectiveTrailComponent|ToggleObjectiveTrail (Variables|Default|GetCreative_ObjectiveTrail_Component) true self Instigator (Variables|Knobs|GetKnob_TeamFilter) (Variables|Knobs|GetKnob_bShowObjectiveTrailFriendly) false (Variables|Knobs|GetKnob_bShowObjectiveTrailInstigator) false 2.0 (Variables|Default|GetCreative_UseRestriction_Component)))

(event Custom|Deactivate%20Objective%20Trail
  (params (Instigator :type Controller%20Object%20Reference))
  ;[1 1 1] Objective Trail Component
  (Class|CreativeObjectiveTrailComponent|ToggleObjectiveTrail (Variables|Default|GetCreative_ObjectiveTrail_Component) false self Instigator (Variables|Knobs|GetKnob_TeamFilter) (Variables|Knobs|GetKnob_bShowObjectiveTrailFriendly) false (Variables|Knobs|GetKnob_bShowObjectiveTrailInstigator) false 2.0 (Variables|Default|GetCreative_UseRestriction_Component)))

(fn ActivateFromOtherDevice
  (params (Instigator :type Controller%20Object%20Reference) (Requester :type ERequesterType%20Enum :default "NewEnumerator0"))
  (return false))

(fn CanInteract
  (params (InteractingPawn :type Fort%20Pawn%20Object%20Reference) (InteractionBeingAttempted :type EInteractionBeingAttempted%20Enum :default "FirstInteraction") (InteractionType :type TInteractionType%20Enum :default "IT_NoInteraction"))
  (return false))

(fn GetDeviceColor
  (return (Utilities|Select :Index (Variables|Knobs|GetKnob_IconColor) :None "(R=0.000000,G=0.537697,B=1.000000,A=1.000000)" :White "(R=1.000000,G=1.000000,B=1.000000,A=1.000000)" :Red "(R=1.000000,G=0.000000,B=0.000000,A=1.000000)" :Orange "(R=1.000000,G=0.500000,B=0.000000,A=1.000000)" :Yellow "(R=1.000000,G=1.000000,B=0.000000,A=1.000000)" :Green "(R=0.000000,G=1.000000,B=0.000000,A=1.000000)" :Teal "(R=0.000000,G=1.000000,B=1.000000,A=1.000000)" :Blue "(R=0.000000,G=0.000000,B=1.000000,A=1.000000)" :Purple "(R=1.000000,G=0.000000,B=1.000000,A=1.000000)")))

(fn GetInstigatingController
  (return 0))

(fn GetInitialMarkerVisibility
  (bind _currentminigame (Variables|/Script/FortniteGame.FortMinigameLogicComponent|GetCurrentMinigame (Variables|Default|GetFortMinigameProgress))
    (:then
      (Minigame|IsInProgress _currentminigame
        (:Yes
          _)
        (:No
          (return true))))
    (:else
      _))
  (return (Math|Conversions|ToBoolean(Integer) (Variables|Knobs|GetKnob_EnabledOnGameStart))))

(fn GetSpawnedPickups
  (return 0))

(fn GetVisibilityComponents
  (bind _visibilitycomponents (|CallParentFunction))
  (Utilities|Array|AppendArray _visibilitycomponents (Utilities|Array|MakeArray (Variables|Debug|GetStaticMeshComponent) (Variables|Default|GetSM_BDA_Attachment_01) (Variables|Default|GetBaseDeviceScreenGlow) (Variables|Default|GetSM_BDA_Attachment-Right) (Variables|Default|GetSM_BDA_Attachment-Left)))
  (return _visibilitycomponents))

(fn IsEnabled
  (return false))

(fn IsHoldingItemForPawn
  (params (Item :type Item%20Variant%20Handle%20Structure) (Pawn :type Pawn%20Object%20Reference))
  (return false))

(fn IsTargetableByLinearMover
  (return true))

(fn SetTriggerCollision
  (params (On :type Boolean))
  (return false))

(fn ShouldInvisiblePropsBeAllowed
  (return true))

(fn ConstructionScript
  (|CallParentFunction)
  (Networking|CanExecuteCosmeticEvents
    (:True
      (Transformation|AttachComponentToComponent (Variables|Default|GetSM_BDA_Attachment_01) (Variables|Debug|GetStaticMeshComponent) "Attachment_Bottom" "KeepRelative" "KeepRelative" "KeepRelative" true)
      (Transformation|AttachComponentToComponent (Variables|Default|GetSM_BDA_Attachment-Right) (Variables|Debug|GetStaticMeshComponent) "Attachment_RightMiddle" "KeepRelative" "KeepRelative" "KeepRelative" true)
      (Transformation|AttachComponentToComponent (Variables|Default|GetSM_BDA_Attachment-Left) (Variables|Debug|GetStaticMeshComponent) "Attachment_LeftMiddle" "KeepRelative" "KeepRelative" "KeepRelative" true)
      (Utilities|SetCreativeFakeGlowParameters (Variables|Default|GetBaseDeviceScreenGlow) (Variables|Default|GetGlowColor) (Variables|Default|GetFlickerIntensity) (Variables|Default|GetFlickerSpeed) (Variables|Default|GetGlowOpacity)
        (:ExecOut
          (Rendering|Material|SetCustomPrimitiveDataFloat (Variables|Debug|GetStaticMeshComponent) 27 17.0))))))
