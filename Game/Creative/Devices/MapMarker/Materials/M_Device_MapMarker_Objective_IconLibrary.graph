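# Translucent UI material for the Creative Map Marker device's objective icon, with a swappable icon drawn over a colored backing and border.
# IconTexture is sampled with scaled, offset, clamped UVs to inset it, and only its alpha is used, adding white.
# BaseTexture channels supply the masks: G tints BackingColor, and inverted R times G tints BorderColor.
# The three layers are summed additively rather than composited over one another.
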Material /Game/Creative/Devices/MapMarker/Materials/M_Device_MapMarker_Objective_IconLibrary
graph {
 properties {blendMode=BLEND_Translucent, materialDomain=MD_UI, pixelDepthOffsetMode=PDOM_Legacy, usedShadingModels=Default Lit}
 nodes [
  n1 = Constant() {r=-0.25}
  n2 = Constant() {r=-0.15}
  n3 = TextureSampleParameter2D(UVs=Add(A=TextureCoordinate() {uTiling=1.5, vTiling=1.5}, B=AppendVector(A=@n1, B=@n2))) {
   parameterName=IconTexture
   samplerSource=SSM_Clamp_WorldGroupSettings
   texture=/Game/Athena/UI/InGame/Creative/PropertyEditor/Icon_Picker/Textures/T_UI_IconPicker_EMPTY
  }
  n4 = VectorParameter() {parameterName=BorderColor, defaultValue=<rgba 1 1 1 1>}
  n5 = TextureSampleParameter2D(UVs=TextureCoordinate()) {parameterName=BaseTexture, samplerType=SAMPLERTYPE_LinearColor, texture=/Game/Creative/Devices/MapMarker/Textures/Device_MapMarker_Icon_A}
  n6 = DotProduct(A=OneMinus(None=@n5.R), B=@n5.G)
  Add(
   A=Add(A=@n3.A, B=LinearInterpolate(B=@n4.RGB, Alpha=@n6))
   B=LinearInterpolate(B=VectorParameter() {parameterName=BackingColor, defaultValue=<rgba 0 0.5 1 1>}.RGB, Alpha=@n5.G)
  )
  Reroute(None=@n5.G)
 ]
}
